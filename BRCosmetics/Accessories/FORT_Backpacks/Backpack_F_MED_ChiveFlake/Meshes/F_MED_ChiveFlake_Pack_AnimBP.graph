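# Animation Blueprint for the ChiveFlake backpack cosmetic, playing idle and a one-shot reaction when an enemy player is eliminated.
# bIsInFrontEnd blends between that state machine and a front-end one, which reuses the elimination clip on timed triggers enabled by the backpack's part modifier.
# The pose then gets an additive pack-bone offset and chained spring-limited AnimDynamics on arm and leg bones, with only the right arm simulating root-relative.
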AnimBlueprint /BRCosmetics/Accessories/FORT_Backpacks/Backpack_F_MED_ChiveFlake/Meshes/F_MED_ChiveFlake_Pack_AnimBP
target_skeleton=&/F_MED_ChiveFlake_Pack_Skeleton
graphs [
 AnimGraph {nodes [
   Input Pose()
   AnimGraphNode_StateMachine_0 = FE Idle + Reactive() {is=title AnimGraphNode_StateMachine}
   AnimGraphNode_StateMachine_1 = Idle and Elimination() {is=title AnimGraphNode_StateMachine}
   n1 = Get bIsInFrontEnd()
   n2 = "Transform (Modify) Bone - Bone: pack"(
    ComponentPose=Local To Component(LocalPose=Blend Poses by bool(BlendPose_0=@AnimGraphNode_StateMachine_0, BlendPose_1=@AnimGraphNode_StateMachine_1, bActiveValue=@n1) {blendType=Linear})
    Translation="0,2.5,0"
    Rotation="0,3,-5.3"
   ) {
    rotationMode=BMM_Additive
    translationMode=BMM_Additive
    alphaBoolBlend {blendOption=Linear}
    boneToModify {boneName=pack}
    rotation {roll=-5.3, yaw=3}
    translation {y=2.5}
   }
   n3 = Anim Dynamics - Bone: shoulder_l!(ComponentPose=@n2, Alpha=1, GravityOverride:Vector="0,0,-980", ExternalForce="0,0,0") -> (Pose) {
    angularDampingOverride=1
    angularSpringConstant=10000
    bAngularSpring=true
    bOverrideAngularDamping=true
    bUseGravityOverride=true
    gravityScale=0.35
    alphaBoolBlend {blendOption=Linear}
    boundBone {boneName=shoulder_l}
    componentAppliedLinearAccClamp=<9000 9000 9000>
    componentLinearAccScale=<1 1 1>
    componentLinearVelScale=<1 1 1>
    gravityOverride {z=-980}
    physicsBodyDefinitions [{
      boundBone {boneName=shoulder_l}
      boxExtents=<20 30 30>
      constraintSetup {angularLimitsMax=<0 5 20>, angularLimitsMin=<0 -30 -20>, angularTarget=<1 0 0>}
      localJointOffset=<4 0 0>
     }]
   }
   n4 = Anim Dynamics - Bone: elbow_l!(ComponentPose=@n3, Alpha=1, GravityOverride:Vector="0,0,-980", ExternalForce="0,0,0") -> (Pose) {
    angularDampingOverride=1
    angularSpringConstant=10000
    bAngularSpring=true
    bOverrideAngularDamping=true
    bUseGravityOverride=true
    gravityScale=0.35
    alphaBoolBlend {blendOption=Linear}
    boundBone {boneName=elbow_l}
    componentAppliedLinearAccClamp=<9000 9000 9000>
    componentLinearAccScale=<1 1 1>
    componentLinearVelScale=<1 1 1>
    gravityOverride {z=-980}
    physicsBodyDefinitions [{
      boundBone {boneName=elbow_l}
      boxExtents=<20 30 30>
      constraintSetup {angularLimitsMax=<0 5 20>, angularLimitsMin=<0 -30 -20>, angularTarget=<1 0 0>}
      localJointOffset=<4 0 0>
     }]
   }
   n5 = Anim Dynamics - Bone: wrist_l!(ComponentPose=@n4, Alpha=1, GravityOverride:Vector="0,0,-980", ExternalForce="0,0,0") -> (Pose) {
    angularDampingOverride=1
    angularSpringConstant=10000
    bAngularSpring=true
    bOverrideAngularDamping=true
    bUseGravityOverride=true
    gravityScale=0.35
    alphaBoolBlend {blendOption=Linear}
    boundBone {boneName=wrist_l}
    componentAppliedLinearAccClamp=<9000 9000 9000>
    componentLinearAccScale=<1 1 1>
    componentLinearVelScale=<1 1 1>
    gravityOverride {z=-980}
    physicsBodyDefinitions [{
      boundBone {boneName=wrist_l}
      boxExtents=<20 30 30>
      constraintSetup {angularLimitsMax=<0 5 20>, angularLimitsMin=<0 -30 -20>, angularTarget=<1 0 0>}
      localJointOffset=<4 0 0>
     }]
   }
   n6 = Anim Dynamics - Bone: shoulder_r!(ComponentPose=@n5, Alpha=1, GravityOverride:Vector="0,0,-800", ExternalForce="0,0,0") -> (Pose) {
    angularDampingOverride=1
    angularSpringConstant=10000
    bAngularSpring=true
    bOverrideAngularDamping=true
    bUseGravityOverride=true
    gravityScale=0.35
    simulationSpace=RootRelative
    alphaBoolBlend {blendOption=Linear}
    boundBone {boneName=shoulder_r}
    componentAppliedLinearAccClamp=<9000 9000 9000>
    componentLinearAccScale=<1 1 1>
    componentLinearVelScale=<1 1 1>
    gravityOverride {z=-800}
    physicsBodyDefinitions [{
      boundBone {boneName=shoulder_r}
      boxExtents=<20 30 30>
      constraintSetup {angularLimitsMax=<0 5 20>, angularLimitsMin=<0 -30 -20>, angularTarget=<1 0 0>}
      localJointOffset=<4 0 0>
     }]
   }
   n7 = Anim Dynamics - Bone: elbow_r!(ComponentPose=@n6, Alpha=1, GravityOverride:Vector="0,0,-800", ExternalForce="0,0,0") -> (Pose) {
    angularDampingOverride=1
    angularSpringConstant=10000
    bAngularSpring=true
    bOverrideAngularDamping=true
    bUseGravityOverride=true
    gravityScale=0.35
    simulationSpace=RootRelative
    alphaBoolBlend {blendOption=Linear}
    boundBone {boneName=elbow_r}
    componentAppliedLinearAccClamp=<9000 9000 9000>
    componentLinearAccScale=<1 1 1>
    componentLinearVelScale=<1 1 1>
    gravityOverride {z=-800}
    physicsBodyDefinitions [{
      boundBone {boneName=elbow_r}
      boxExtents=<20 30 30>
      constraintSetup {angularLimitsMax=<0 5 20>, angularLimitsMin=<0 -30 -20>, angularTarget=<1 0 0>}
      localJointOffset=<4 0 0>
     }]
   }
   n8 = Anim Dynamics - Bone: wrist_r!(ComponentPose=@n7, Alpha=1, GravityOverride:Vector="0,0,-800", ExternalForce="0,0,0") -> (Pose) {
    angularDampingOverride=1
    angularSpringConstant=10000
    bAngularSpring=true
    bOverrideAngularDamping=true
    bUseGravityOverride=true
    gravityScale=0.35
    simulationSpace=RootRelative
    alphaBoolBlend {blendOption=Linear}
    boundBone {boneName=wrist_r}
    componentAppliedLinearAccClamp=<9000 9000 9000>
    componentLinearAccScale=<1 1 1>
    componentLinearVelScale=<1 1 1>
    gravityOverride {z=-800}
    physicsBodyDefinitions [{
      boundBone {boneName=wrist_r}
      boxExtents=<20 30 30>
      constraintSetup {angularLimitsMax=<0 5 20>, angularLimitsMin=<0 -30 -20>, angularTarget=<1 0 0>}
      localJointOffset=<4 0 0>
     }]
   }
   n9 = Anim Dynamics - Bone: leg_l!(ComponentPose=@n8, Alpha=1, GravityOverride:Vector="0,0,-980", ExternalForce="0,0,0") -> (Pose) {
    angularDampingOverride=1
    angularSpringConstant=20000
    bAngularSpring=true
    bOverrideAngularDamping=true
    bUseGravityOverride=true
    gravityScale=0.35
    alphaBoolBlend {blendOption=Linear}
    boundBone {boneName=leg_l}
    componentAppliedLinearAccClamp=<9000 9000 9000>
    componentLinearAccScale=<1 1 1>
    componentLinearVelScale=<1 1 1>
    gravityOverride {z=-980}
    physicsBodyDefinitions [{
      boundBone {boneName=leg_l}
      boxExtents=<20 30 30>
      constraintSetup {angularLimitsMax=<0 20 20>, angularLimitsMin=<0 -20 -20>, angularTarget=<1 0 0>}
      localJointOffset=<4 0 0>
     }]
   }
   n10 = Anim Dynamics - Bone: knee_l!(ComponentPose=@n9, Alpha=1, GravityOverride:Vector="0,0,-980", ExternalForce="0,0,0") -> (Pose) {
    angularDampingOverride=1
    angularSpringConstant=20000
    bAngularSpring=true
    bOverrideAngularDamping=true
    bUseGravityOverride=true
    gravityScale=0.35
    alphaBoolBlend {blendOption=Linear}
    boundBone {boneName=knee_l}
    componentAppliedLinearAccClamp=<9000 9000 9000>
    componentLinearAccScale=<1 1 1>
    componentLinearVelScale=<1 1 1>
    gravityOverride {z=-980}
    physicsBodyDefinitions [{
      boundBone {boneName=knee_l}
      boxExtents=<20 30 30>
      constraintSetup {angularLimitsMax=<0 20 20>, angularLimitsMin=<0 -20 -20>, angularTarget=<1 0 0>}
      localJointOffset=<4 0 0>
     }]
   }
   n11 = Anim Dynamics - Bone: leg_r!(ComponentPose=@n10, Alpha=1, GravityOverride:Vector="0,0,-980", ExternalForce="0,0,0") -> (Pose) {
    angularDampingOverride=1
    angularSpringConstant=20000
    bAngularSpring=true
    bOverrideAngularDamping=true
    bUseGravityOverride=true
    gravityScale=0.35
    alphaBoolBlend {blendOption=Linear}
    boundBone {boneName=leg_r}
    componentAppliedLinearAccClamp=<9000 9000 9000>
    componentLinearAccScale=<1 1 1>
    componentLinearVelScale=<1 1 1>
    gravityOverride {z=-980}
    physicsBodyDefinitions [{
      boundBone {boneName=leg_r}
      boxExtents=<20 30 30>
      constraintSetup {angularLimitsMax=<0 20 20>, angularLimitsMin=<0 -20 -20>, angularTarget=<1 0 0>}
      localJointOffset=<4 0 0>
     }]
   }
   n12 = Anim Dynamics - Bone: knee_r!(ComponentPose=@n11, Alpha=1, GravityOverride:Vector="0,0,-980", ExternalForce="0,0,0") -> (Pose) {
    angularDampingOverride=1
    angularSpringConstant=20000
    bAngularSpring=true
    bOverrideAngularDamping=true
    bUseGravityOverride=true
    gravityScale=0.35
    alphaBoolBlend {blendOption=Linear}
    boundBone {boneName=knee_r}
    componentAppliedLinearAccClamp=<9000 9000 9000>
    componentLinearAccScale=<1 1 1>
    componentLinearVelScale=<1 1 1>
    gravityOverride {z=-980}
    physicsBodyDefinitions [{
      boundBone {boneName=knee_r}
      boxExtents=<20 30 30>
      constraintSetup {angularLimitsMax=<0 20 20>, angularLimitsMin=<0 -20 -20>, angularTarget=<1 0 0>}
      localJointOffset=<4 0 0>
     }]
   }
   n13 = Anim Dynamics - Bone: ankle_r!(ComponentPose=@n12, Alpha=1, GravityOverride:Vector="0,0,-980", ExternalForce="0,0,0") -> (Pose) {
    angularDampingOverride=1
    angularSpringConstant=20000
    bAngularSpring=true
    bOverrideAngularDamping=true
    bUseGravityOverride=true
    gravityScale=0.35
    alphaBoolBlend {blendOption=Linear}
    boundBone {boneName=ankle_r}
    componentAppliedLinearAccClamp=<9000 9000 9000>
    componentLinearAccScale=<1 1 1>
    componentLinearVelScale=<1 1 1>
    gravityOverride {z=-980}
    physicsBodyDefinitions [{
      boundBone {boneName=ankle_r}
      boxExtents=<20 30 30>
      constraintSetup {angularLimitsMax=<0 20 20>, angularLimitsMin=<0 -20 -20>, angularTarget=<1 0 0>}
      localJointOffset=<4 0 0>
     }]
   }
   n14 = Anim Dynamics - Bone: ankle_l!(ComponentPose=@n13, Alpha=1, GravityOverride:Vector="0,0,-980", ExternalForce="0,0,0") -> (Pose) {
    angularDampingOverride=1
    angularSpringConstant=20000
    bAngularSpring=true
    bOverrideAngularDamping=true
    bUseGravityOverride=true
    gravityScale=0.35
    alphaBoolBlend {blendOption=Linear}
    boundBone {boneName=ankle_l}
    componentAppliedLinearAccClamp=<9000 9000 9000>
    componentLinearAccScale=<1 1 1>
    componentLinearVelScale=<1 1 1>
    gravityOverride {z=-980}
    physicsBodyDefinitions [{
      boundBone {boneName=ankle_l}
      boxExtents=<20 30 30>
      constraintSetup {angularLimitsMax=<0 20 20>, angularLimitsMin=<0 -20 -20>, angularTarget=<1 0 0>}
      localJointOffset=<4 0 0>
     }]
   }
   Output Pose(Result=Component To Local(ComponentPose=@n14)) {layerGroup=None}
  ]}
 FE Idle + Reactive {
  entry=FE_Idle
  path=AnimGraph.AnimGraphNode_StateMachine_0.FE Idle + Reactive
  ids=id
  nodes [
   AnimStateNode_0 = FE_Idle() {is=title AnimStateNode}
   AnimStateNode_1 = FE_Reactive() {is=title AnimStateNode, stateEntered {guid=#1, notifyName=Entered FE_Reactive_Anim}}
   AnimStateTransitionNode_0 = AnimStateTransitionNode()
   AnimStateTransitionNode_1 = AnimStateTransitionNode() {bAutomaticRuleBasedOnSequencePlayerInState=true}
  ]
  links [
   {
    dst=FE_Reactive
    rule=AnimGraph.AnimGraphNode_StateMachine_0.FE Idle + Reactive.AnimStateTransitionNode_0.Transition
    src=FE_Idle
    via=AnimStateTransitionNode_0
   }
   {
    dst=FE_Idle
    rule=AnimGraph.AnimGraphNode_StateMachine_0.FE Idle + Reactive.AnimStateTransitionNode_1.Transition
    src=FE_Reactive
    via=AnimStateTransitionNode_1
   }
  ]
 }
 FE_Idle {
  path=AnimGraph.AnimGraphNode_StateMachine_0.FE Idle + Reactive.AnimStateNode_0.FE_Idle
  nodes [Output Animation Pose(Result=Sequence Player 'ChiveFlake_Idle_PACK'() {sequence=/BRCosmetics/Accessories/FORT_Backpacks/Backpack_F_MED_ChiveFlake/Animations/ChiveFlake_Idle_PACK})]
 }
 FE_Reactive {
  path=AnimGraph.AnimGraphNode_StateMachine_0.FE Idle + Reactive.AnimStateNode_1.FE_Reactive
  nodes [
   n1 = Sequence Player 'ChiveFlake_Elimination_PACK'() {bLoopAnimation=false, sequence=/BRCosmetics/Accessories/FORT_Backpacks/Backpack_F_MED_ChiveFlake/Animations/ChiveFlake_Elimination_PACK}
   Output Animation Pose(Result=@n1)
  ]
 }
 Transition {
  path=AnimGraph.AnimGraphNode_StateMachine_0.FE Idle + Reactive.AnimStateTransitionNode_0.Transition
  nodes [Result(bCanEnterTransition=Get Play Reactive FE Anim() -> (Play Reactive FE Anim:bool))]
 }
 Transition {path=AnimGraph.AnimGraphNode_StateMachine_0.FE Idle + Reactive.AnimStateTransitionNode_1.Transition, nodes [Result()]}
 Idle and Elimination {
  entry=Idle
  path=AnimGraph.AnimGraphNode_StateMachine_1.Idle and Elimination
  ids=id
  nodes [
   AnimStateNode_0 = Idle() {
    stateEntered {guid=#2}
    stateFullyBlended {guid=#3}
    stateLeft {guid=#4}
   }
   AnimStateNode_1 = Elimination() {
    is=title AnimStateNode
    stateEntered {guid=#5}
    stateFullyBlended {guid=#6}
    stateLeft {guid=#7}
   }
   AnimStateTransitionNode_0 = AnimStateTransitionNode() {
    transitionEnd {guid=#8}
    transitionInterrupt {guid=#9}
    transitionStart {guid=#10}
   }
   AnimStateTransitionNode_1 = AnimStateTransitionNode() {bAutomaticRuleBasedOnSequencePlayerInState=true}
  ]
  links [
   {
    dst=Idle
    rule=AnimGraph.AnimGraphNode_StateMachine_1.Idle and Elimination.AnimStateTransitionNode_1.Transition
    src=Elimination
    via=AnimStateTransitionNode_1
   }
   {
    dst=Elimination
    rule=AnimGraph.AnimGraphNode_StateMachine_1.Idle and Elimination.AnimStateTransitionNode_0.Transition
    src=Idle
    via=AnimStateTransitionNode_0
   }
  ]
 }
 Idle {
  path=AnimGraph.AnimGraphNode_StateMachine_1.Idle and Elimination.AnimStateNode_0.Idle
  nodes [
   Sequence Player 'GuineaPig_Idle_PACK'() {sequence=/BRCosmetics/Accessories/FORT_Backpacks/Backpack_M_MED_GuineaPig/Animations/GuineaPig_Idle_PACK}
   Output Animation Pose(Result=Sequence Player 'ChiveFlake_Idle_PACK'() {sequence=&ChiveFlake_Idle_PACK})
  ]
 }
 Elimination {
  path=AnimGraph.AnimGraphNode_StateMachine_1.Idle and Elimination.AnimStateNode_1.Elimination
  nodes [
   Sequence Player 'GuineaPig_TakeDamage_PACK'() {bLoopAnimation=false, sequence=/BRCosmetics/Accessories/FORT_Backpacks/Backpack_M_MED_GuineaPig/Animations/GuineaPig_TakeDamage_PACK}
   n1 = Sequence Player 'ChiveFlake_Elimination_PACK'() {bLoopAnimation=false, sequence=&ChiveFlake_Elimination_PACK}
   Output Animation Pose(Result=@n1)
  ]
 }
 Transition {
  path=AnimGraph.AnimGraphNode_StateMachine_1.Idle and Elimination.AnimStateTransitionNode_0.Transition
  nodes [Result(bCanEnterTransition=Get bEnemyPlayerEliminated() -> (bEnemyPlayerEliminated:bool))]
 }
 Transition {path=AnimGraph.AnimGraphNode_StateMachine_1.Idle and Elimination.AnimStateTransitionNode_1.Transition, nodes [Result()]}
 EventGraph {nodes [
   Event BlueprintUpdateAnimation()
   > n1 = Sequence!(execute) -> (then_0, then_1, then_2:Exec)
   > n2 = Gravity Override State Function DEPRECATED(CharacterPartAnimInstance=Self-Reference(), GameTimeInSeconds=GetGameTimeInSeconds(), GravityOverrideParams=Get Gravity Override Params())
   > Set GravityOverride(GravityOverride=@n2)
   Set AnimCurve(
    execute=@n1.then_1
    AnimCurve:double=Remap Float(Input=GetCurveValue(CurveName=DynamicsCurve), OldRange=(X=0,Y=1), NewRange=(X=1,Y=0)) -> (Output=~)
   ) -> (Output_Get:double)
   Event OnCharacterCustomizationLoaded()
   > n3 = Cast To FortPlayerPawn(Object=TryGetFortPawn())
   > n4 = Cast To B_Athena_PartModifier_Backpack_ChiveFlake!(
    execute
    Object=GetModifierForPartType!(self:Fort Player Pawn Object Reference=@n3, PartType:EFortCustomPartType Enum=Backpack) -> (ReturnValue:Custom Character Part Modifier Object Reference)
   ) -> (then, CastFailed, AsB Athena Part Modifier Backpack Chive Flake:B Athena Part Modifier Backpack Chive Flake Object Reference) {is=title K2Node_DynamicCast}
   > n5 = Branch(Condition=Get FE_ReactivityEnabled!(self:B Athena Part Modifier Backpack Chive Flake Object Reference=@n4) -> (FE_ReactivityEnabled:bool))
   > Do Once(Reset=@n7.then)
   > n6 = Delay(Duration=4)
   Set Play Reactive FE Anim(execute=[@n9.then, @n6.then], Play Reactive FE Anim:bool=true) -> (Output_Get:bool)
   n7 = Set Play Reactive FE Anim(execute=@n5.else, Play Reactive FE Anim=false)
   n8 = Event BlueprintProcessAnimInputEvent()
   > Branch(Condition=IsNamed(self=@n8.AnimInputEvent, InEventName=aie.EnemyPlayerEliminated))
   > Set bEnemyPlayerEliminated(bEnemyPlayerEliminated:bool=true) -> (Output_Get:bool)
   > Delay(Duration=0.2)
   > Set bEnemyPlayerEliminated(bEnemyPlayerEliminated=false)
   Event BlueprintPostEvaluateAnimation() {is=title K2Node_Event}
   > Set bEnemyPlayerEliminated(bEnemyPlayerEliminated=false)
   > Set Play Reactive FE Anim(Play Reactive FE Anim=false)
   AnimNotify_ResetReactivity() {is=title K2Node_Event}
   > Set Play Reactive FE Anim(Play Reactive FE Anim=false)
   > n9 = Delay(Duration=10)
   Branch(Condition=Get FrontEnd() -> (FrontEnd:bool))
   > Set Play Reactive FE Anim(Play Reactive FE Anim=true)
   n10 = Branch(Condition=IsCosmeticObjectDisplayedInFrontendContext!(CosmeticObject:Object Reference=Self-Reference()) -> (ReturnValue:bool))
   > Set FrontEnd(FrontEnd:bool=true) -> (Output_Get:bool)
   Set FrontEnd(execute=@n10.else, FrontEnd=false)
   TryGetPawnOwner()
   IsCosmeticObjectDisplayedInFrontendContext()
  ]}
]
variables [
 AnimCurve {type=number}
 bEnemyPlayerEliminated {type=boolean}
 FrontEnd {type=boolean}
 Gravity Override Params {type=GravityOverrideParamsStruct}
 GravityOverride {type=Vector}
 Play Reactive FE Anim {type=boolean}
]
